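# Physics asset for the KelpLinen Keytar instrument cosmetic, defining a jointed chain with a symbol charm attached to the handle.
# A handle capsule anchors a small sphere root, followed by four capsule links and a capsule for the end symbol.
# Constraints join neighbors in sequence with varying swing limits.
# Twist is locked along the chain except at the symbol, which has a narrow twist limit.
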PhysicsAsset /SparksCosmetics/Instrument/Keytar/KelpLinen/Meshes/KelpLinen_Keytar_Physics
bodies {
 chain_root [{shapeType=Sphere, radius=0.505, shapeName=chain_root_sphere}]
 handle [{shapeType=Capsule, length=16.0801, radius=3.0049, center=<0.7174 1.3219 54.408>, rotation=<rot 0.7242 -90.0071 0.2938>}]
 link_01 [{shapeType=Capsule, length=1, radius=1, shapeName=link_01_capsule, center=<1 0 0>, rotation=<rot -88.1126 179.9988 -179.9989>}]
 link_02 [{shapeType=Capsule, length=1, radius=1, shapeName=link_02_capsule, center=<2 0 0>, rotation=<rot 90 0 0>}]
 link_03 [{shapeType=Capsule, length=1, radius=1, shapeName=link_03_capsule, center=<1 0 0>, rotation=<rot 90 0 0>}]
 link_04 [{shapeType=Capsule, length=1, radius=1, shapeName=link_04_capsule, rotation=<rot 90 0 0>}]
 symbol [{shapeType=Capsule, length=1, radius=1.7176, shapeName=symbol_capsule, center=<1.3 0 0>, rotation=<rot 90 0 0>}]
}
constraints [
 {bone1Name=chain_root, bone2Name=handle, swing1LimitDegrees=45, swing2LimitDegrees=45, twistLimitDegrees=45, twistMotion=Locked}
 {bone1Name=link_01, bone2Name=chain_root, swing1LimitDegrees=25, swing2LimitDegrees=25, twistLimitDegrees=45, twistMotion=Locked}
 {bone1Name=link_02, bone2Name=link_01, swing1LimitDegrees=65, swing2LimitDegrees=65, twistLimitDegrees=45, twistMotion=Locked}
 {bone1Name=link_03, bone2Name=link_02, swing1LimitDegrees=45, swing2LimitDegrees=90, twistLimitDegrees=45, twistMotion=Locked}
 {bone1Name=link_04, bone2Name=link_03, swing1LimitDegrees=45, swing2LimitDegrees=90, twistLimitDegrees=45, twistMotion=Locked}
 {bone1Name=symbol, bone2Name=link_04, swing1LimitDegrees=45, swing2LimitDegrees=45, twistLimitDegrees=5}
]
